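# Physics asset for the October character's skeleton, giving it capsule collision bodies and joints for physics simulation.
# Capsules cover the spine, head, face and mouth roots, limbs, feet, some finger bones, plus helper bones for scarf, ties and ribbon.
# Constraints link them in a tree from the pelvis, with scarf pieces on the clavicles, ties on spine_03 and ribbon on the head.
# Every joint shares the same uniform swing and twist limit angles.
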PhysicsAsset /DealershipSharedLibrary/Assets/Sets/October/October_PhysicsAsset
bodies {
 ball_l [{shapeType=Capsule, length=40.3361, radius=29.7385, center=<18.9373 -14.599 1.6443>, rotation=<rot 3.8979 -90 -2.7678>}]
 ball_r [{shapeType=Capsule, length=42.8556, radius=36.7227, center=<-9.1706 15.2006 1.2006>, rotation=<rot 1.3526 -90 10.7376>}]
 bone_face_root [{shapeType=Capsule, length=35.5861, radius=21.0354, center=<7.6195 38.4265 -0.0265>, rotation=<rot -0.037 -90 0.3004>}]
 bone_mouth_root [{shapeType=Capsule, length=27.2333, radius=23.4254, center=<-1.2806 -18.5502 10.227>, rotation=<rot 0 92.6039 -85.248>}]
 calf_l [{shapeType=Capsule, length=68.3605, radius=50.9358, center=<-28.9784 4.0511 3.6664>, rotation=<rot 0 -46.474 -83.8802>}]
 calf_r [{shapeType=Capsule, length=68.1884, radius=50.8382, center=<28.3949 -4.882 -3.53>, rotation=<rot 0 -42.1218 -82.6821>}]
 clavicle_l [{shapeType=Capsule, length=80.4117, radius=69.6958, center=<32.4609 -9.1959 -4.0118>, rotation=<rot 0 31.6817 -87.6994>}]
 clavicle_r [{shapeType=Capsule, length=76.9603, radius=62.5846, center=<-40.7034 7.2207 3.1363>, rotation=<rot 0 -153.3355 -88.2959>}]
 foot_l [{shapeType=Capsule, length=43.1756, radius=43.0052, center=<-7.8067 15.7666 -3.5348>, rotation=<rot -20.1836 -90 -3.1775>}]
 foot_r [{shapeType=Capsule, length=43.5966, radius=35.0791, center=<11.8622 -9.6934 -10.2036>, rotation=<rot 30.6818 -90 -2.3756>}]
 hand_l [{shapeType=Capsule, length=26.4773, radius=20.7767, center=<11.8521 -2.7108 3.4618>, rotation=<rot 5.1895 -90 -0.3882>}]
 hand_r [{shapeType=Capsule, length=28.1028, radius=21.5418, center=<-12.0986 2.7516 -2.3034>, rotation=<rot 9.7356 -90 4.3688>}]
 head [{shapeType=Capsule, length=85.659, radius=71.574, center=<55.2193 -0.9782 11.3077>, rotation=<rot 0 40.4922 -53.2786>}]
 helper_l_tie1 [{shapeType=Capsule, length=22.0758, radius=9.3653, center=<20.1263 0.8528 -0.7311>, rotation=<rot 0 -107.5115 -86.4187>}]
 helper_r_tie1 [{shapeType=Capsule, length=22.0469, radius=9.3552, center=<-20.1932 -0.9712 0.791>, rotation=<rot 0 -106.8956 -85.673>}]
 helper_ribbon [{shapeType=Capsule, length=43.5673, radius=9.6731, center=<44.0545 6.9014 0.7598>, rotation=<rot 0 106.5725 -88.9538>}]
 helper_scarf_l_a [{shapeType=Capsule, length=38.8453, radius=24.285, center=<16.8287 1.6527 -0.1153>, rotation=<rot -11.7472 -90 -37.6058>}]
 helper_scarf_r_a [{shapeType=Capsule, length=38.8453, radius=24.2849, center=<-16.8268 -1.6536 0.1152>, rotation=<rot -11.7473 -90 -37.6057>}]
 index_01_l [{shapeType=Capsule, length=13.6653, radius=8.602, center=<12.5576 1.6729 -0.7065>, rotation=<rot 0 82.0231 -82.6334>}]
 index_01_r [{shapeType=Capsule, length=13.6656, radius=8.6019, center=<-12.5576 -1.6737 0.7068>, rotation=<rot 0 82.0222 -82.6318>}]
 lowerarm_l [{shapeType=Capsule, length=44.8255, radius=41.6251, center=<36.9157 5.4986 -2.8457>, rotation=<rot 0 94.1994 -83.5365>}]
 lowerarm_r [{shapeType=Capsule, length=45.3382, radius=41.6304, center=<-36.205 -6.0586 2.0894>, rotation=<rot 0 95.7411 -81.8269>}]
 middle_01_l [{shapeType=Capsule, length=12.4774, radius=9.4324, center=<12.4481 -0.1055 0.1246>, rotation=<rot 0 -99.6323 -89.4439>}]
 middle_01_r [{shapeType=Capsule, length=12.4749, radius=9.4248, center=<-12.4588 0.0971 -0.1228>, rotation=<rot 0 -99.5229 -89.4311>}]
 pelvis [{shapeType=Capsule, length=124.9086, radius=101.2398, center=<12.4211 0.8353 -0.4904>, rotation=<rot 2.4204 -90 -1.6627>}]
 ring_01_l [{shapeType=Capsule, length=11.7674, radius=9.315, center=<11.048 -1.7726 0.4153>, rotation=<rot 0 -108.4461 -88.5425>}]
 ring_01_r [{shapeType=Capsule, length=11.7676, radius=9.315, center=<-11.0473 1.7722 -0.4149>, rotation=<rot 0 -108.4462 -88.5427>}]
 spine_01 [{shapeType=Capsule, length=124.097, radius=64.2591, center=<12.6703 55.8246 4.494>, rotation=<rot -4.106 -90 -1.0912>}]
 spine_02 [{shapeType=Capsule, length=120.7412, radius=85.5953, center=<-10.0703 -13.2557 0.9796>, rotation=<rot -3.8143 -90 -0.3549>}]
 spine_03 [{shapeType=Capsule, length=78.2008, radius=56.1203, center=<9.5647 10.3862 7.0207>, rotation=<rot 0 -166.6748 -89.8976>}]
 thigh_l [{shapeType=Capsule, length=72.8103, radius=62.7015, center=<-34.2002 -2.9212 -0.1994>, rotation=<rot 0 -3.9822 -77.5159>}]
 thigh_r [{shapeType=Capsule, length=73.3515, radius=61.8747, center=<36.7964 3.9485 2.1831>, rotation=<rot 0 1.4364 -85.3658>}]
 thumb_02_l [{shapeType=Capsule, length=15.5381, radius=9.7137, center=<11.5146 0.542 1.159>, rotation=<rot 0 -94.6987 -87.9957>}]
 thumb_02_r [{shapeType=Capsule, length=14.015, radius=9.4988, center=<-13.0296 -1.1084 -0.074>, rotation=<rot 0 -89.5582 -82.7143>}]
 upperarm_l [{shapeType=Capsule, length=53.8041, radius=50.6818, center=<20.6732 9.0566 -3.2578>, rotation=<rot 38.2094 -90 -5.8107>}]
 upperarm_r [{shapeType=Capsule, length=53.0125, radius=50.3934, center=<-21.8286 -7.6502 4.4456>, rotation=<rot 40.7471 -90 1.0344>}]
}
constraints [
 {bone1Name=spine_01, bone2Name=pelvis, swing1LimitDegrees=45, swing2LimitDegrees=45, twistLimitDegrees=45}
 {bone1Name=spine_02, bone2Name=spine_01, swing1LimitDegrees=45, swing2LimitDegrees=45, twistLimitDegrees=45}
 {bone1Name=spine_03, bone2Name=spine_02, swing1LimitDegrees=45, swing2LimitDegrees=45, twistLimitDegrees=45}
 {bone1Name=clavicle_l, bone2Name=spine_03, swing1LimitDegrees=45, swing2LimitDegrees=45, twistLimitDegrees=45}
 {bone1Name=upperarm_l, bone2Name=clavicle_l, swing1LimitDegrees=45, swing2LimitDegrees=45, twistLimitDegrees=45}
 {bone1Name=lowerarm_l, bone2Name=upperarm_l, swing1LimitDegrees=45, swing2LimitDegrees=45, twistLimitDegrees=45}
 {bone1Name=hand_l, bone2Name=lowerarm_l, swing1LimitDegrees=45, swing2LimitDegrees=45, twistLimitDegrees=45}
 {bone1Name=index_01_l, bone2Name=hand_l, swing1LimitDegrees=45, swing2LimitDegrees=45, twistLimitDegrees=45}
 {bone1Name=middle_01_l, bone2Name=hand_l, swing1LimitDegrees=45, swing2LimitDegrees=45, twistLimitDegrees=45}
 {bone1Name=ring_01_l, bone2Name=hand_l, swing1LimitDegrees=45, swing2LimitDegrees=45, twistLimitDegrees=45}
 {bone1Name=thumb_02_l, bone2Name=hand_l, swing1LimitDegrees=45, swing2LimitDegrees=45, twistLimitDegrees=45}
 {bone1Name=helper_scarf_l_a, bone2Name=clavicle_l, swing1LimitDegrees=45, swing2LimitDegrees=45, twistLimitDegrees=45}
 {bone1Name=clavicle_r, bone2Name=spine_03, swing1LimitDegrees=45, swing2LimitDegrees=45, twistLimitDegrees=45}
 {bone1Name=upperarm_r, bone2Name=clavicle_r, swing1LimitDegrees=45, swing2LimitDegrees=45, twistLimitDegrees=45}
 {bone1Name=lowerarm_r, bone2Name=upperarm_r, swing1LimitDegrees=45, swing2LimitDegrees=45, twistLimitDegrees=45}
 {bone1Name=hand_r, bone2Name=lowerarm_r, swing1LimitDegrees=45, swing2LimitDegrees=45, twistLimitDegrees=45}
 {bone1Name=index_01_r, bone2Name=hand_r, swing1LimitDegrees=45, swing2LimitDegrees=45, twistLimitDegrees=45}
 {bone1Name=middle_01_r, bone2Name=hand_r, swing1LimitDegrees=45, swing2LimitDegrees=45, twistLimitDegrees=45}
 {bone1Name=ring_01_r, bone2Name=hand_r, swing1LimitDegrees=45, swing2LimitDegrees=45, twistLimitDegrees=45}
 {bone1Name=thumb_02_r, bone2Name=hand_r, swing1LimitDegrees=45, swing2LimitDegrees=45, twistLimitDegrees=45}
 {bone1Name=helper_scarf_r_a, bone2Name=clavicle_r, swing1LimitDegrees=45, swing2LimitDegrees=45, twistLimitDegrees=45}
 {bone1Name=head, bone2Name=spine_03, swing1LimitDegrees=45, swing2LimitDegrees=45, twistLimitDegrees=45}
 {bone1Name=bone_face_root, bone2Name=head, swing1LimitDegrees=45, swing2LimitDegrees=45, twistLimitDegrees=45}
 {bone1Name=bone_mouth_root, bone2Name=bone_face_root, swing1LimitDegrees=45, swing2LimitDegrees=45, twistLimitDegrees=45}
 {bone1Name=helper_ribbon, bone2Name=head, swing1LimitDegrees=45, swing2LimitDegrees=45, twistLimitDegrees=45}
 {bone1Name=helper_l_tie1, bone2Name=spine_03, swing1LimitDegrees=45, swing2LimitDegrees=45, twistLimitDegrees=45}
 {bone1Name=helper_r_tie1, bone2Name=spine_03, swing1LimitDegrees=45, swing2LimitDegrees=45, twistLimitDegrees=45}
 {bone1Name=thigh_l, bone2Name=pelvis, swing1LimitDegrees=45, swing2LimitDegrees=45, twistLimitDegrees=45}
 {bone1Name=calf_l, bone2Name=thigh_l, swing1LimitDegrees=45, swing2LimitDegrees=45, twistLimitDegrees=45}
 {bone1Name=foot_l, bone2Name=calf_l, swing1LimitDegrees=45, swing2LimitDegrees=45, twistLimitDegrees=45}
 {bone1Name=ball_l, bone2Name=foot_l, swing1LimitDegrees=45, swing2LimitDegrees=45, twistLimitDegrees=45}
 {bone1Name=thigh_r, bone2Name=pelvis, swing1LimitDegrees=45, swing2LimitDegrees=45, twistLimitDegrees=45}
 {bone1Name=calf_r, bone2Name=thigh_r, swing1LimitDegrees=45, swing2LimitDegrees=45, twistLimitDegrees=45}
 {bone1Name=foot_r, bone2Name=calf_r, swing1LimitDegrees=45, swing2LimitDegrees=45, twistLimitDegrees=45}
 {bone1Name=ball_r, bone2Name=foot_r, swing1LimitDegrees=45, swing2LimitDegrees=45, twistLimitDegrees=45}
]
